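# Physics asset for the MossyDeport backpack cosmetic, giving its rope and two tassels jointed ragdoll-style bodies for secondary motion.
# A small pack sphere anchors MSTR_rope and both tassel root capsules through constraints with locked swing and twist.
# A three-capsule rope chain and two-capsule left and right tassel chains hang from those roots with limited swing and zero-degree twist limits.
# Right tassel segments use much larger capsules than left ones.
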PhysicsAsset /BRCosmetics/Accessories/FORT_Backpacks/Backpack_M_MED_MossyDeport/Meshes/M_MED_MossyDeport_Pack_Physics_rope
bodies {
 MSTR_rope [{shapeType=Capsule, length=1, radius=1, shapeName=MSTR_rope_capsule}]
 pack [{shapeType=Sphere, radius=0.505, shapeName=pack_sphere}]
 rope_1 [{shapeType=Capsule, length=3, radius=2, shapeName=rope_1_capsule, center=<3 0 0>, rotation=<rot -90 0 0>}]
 rope_2 [{shapeType=Capsule, length=3, radius=2, shapeName=rope_2_capsule, center=<3 0 0>, rotation=<rot -90 0 0>}]
 rope_3 [{shapeType=Capsule, length=3, radius=2, shapeName=rope_3_capsule, center=<1.5 0 0>, rotation=<rot -90 0 0>}]
 tassel_01_l [{shapeType=Capsule, length=1, radius=0.5, shapeName=tassel_01_l_capsule, center=<1 0 0>, rotation=<rot -90 0 0>}]
 tassel_01_r [{shapeType=Capsule, length=3.5, radius=2.5, shapeName=tassel_01_r_capsule, center=<3 0 0>, rotation=<rot -90 0 0>}]
 tassel_02_l [{shapeType=Capsule, length=1, radius=0.5, shapeName=tassel_02_l_capsule, center=<1 0 0>, rotation=<rot -90 0 0>}]
 tassel_02_r [{shapeType=Capsule, length=3.5, radius=2.5, shapeName=tassel_02_r_capsule, center=<3 0 0>, rotation=<rot -90 0 0>}]
 tassel_l [{shapeType=Capsule, length=0.505, radius=0.505, shapeName=tassel_l_capsule, rotation=<rot 0 -90 0>}]
 tassel_r [{shapeType=Capsule, length=0.505, radius=0.505, shapeName=tassel_r_capsule, rotation=<rot 0 -90 0>}]
}
constraints [
 {
  bone1Name=MSTR_rope
  bone2Name=pack
  swing1LimitDegrees=45
  swing1Motion=Locked
  swing2LimitDegrees=45
  swing2Motion=Locked
  twistLimitDegrees=45
  twistMotion=Locked
 }
 {bone1Name=rope_1, bone2Name=MSTR_rope, swing1LimitDegrees=15, swing2LimitDegrees=3, twistLimitDegrees=0}
 {bone1Name=rope_2, bone2Name=rope_1, swing1LimitDegrees=35, swing2LimitDegrees=5, twistLimitDegrees=0}
 {
  bone1Name=tassel_l
  bone2Name=pack
  swing1LimitDegrees=45
  swing1Motion=Locked
  swing2LimitDegrees=45
  swing2Motion=Locked
  twistLimitDegrees=45
  twistMotion=Locked
 }
 {bone1Name=tassel_01_l, bone2Name=tassel_l, swing1LimitDegrees=45, swing2LimitDegrees=45, twistLimitDegrees=0}
 {bone1Name=tassel_02_l, bone2Name=tassel_01_l, swing1LimitDegrees=45, swing2LimitDegrees=45, twistLimitDegrees=0}
 {
  bone1Name=tassel_r
  bone2Name=pack
  swing1LimitDegrees=45
  swing1Motion=Locked
  swing2LimitDegrees=45
  swing2Motion=Locked
  twistLimitDegrees=45
  twistMotion=Locked
 }
 {bone1Name=tassel_01_r, bone2Name=tassel_r, swing1LimitDegrees=40, swing2LimitDegrees=40, twistLimitDegrees=0}
 {bone1Name=tassel_02_r, bone2Name=tassel_01_r, swing1LimitDegrees=40, swing2LimitDegrees=60, twistLimitDegrees=0}
 {bone1Name=rope_3, bone2Name=rope_2, swing1LimitDegrees=35, swing2LimitDegrees=35, twistLimitDegrees=0}
]
