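# Post-process style AnimBP for the MossyDeport backpack that simulates its rope physics and adjusts the pack bone to the wearer's spine bend.
# The cached input pose gets a small pack offset, then bIsInFrontEnd selects between pack-relative front-end and root-relative in-game RigidBody simulations with separate physics assets.
# Each update maps spine_05-to-spine_01 yaw to additive pack rotation and translation for the in-game branch.
# It also stores gravity-override outputs, which the RigidBody nodes show no links to.
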AnimBlueprint /BRCosmetics/Accessories/FORT_Backpacks/Backpack_M_MED_MossyDeport/Meshes/M_MED_MossyDeport_Pack_AnimBP
target_skeleton=&/M_MED_MossyDeport_Pack_Skeleton
graphs [
 AnimGraph {nodes [
   RigidBody(ComponentPose="Transform (Modify) Bone - Bone: pack"(ComponentPose=Local To Component(), Translation="0,-0.5,0", Rotation="0,0,0") {translationMode=BMM_Additive, translationSpace=BCS_BoneSpace, boneToModify {boneName=pack}, translation {y=-0.5}}) {
    bOverrideWorldGravity=true
    overridePhysicsAsset=&/M_MED_MossyDeport_Pack_Physics_ropeFE
    simulationSpace=BaseBoneSpace
    baseBoneRef {boneName=pack}
    componentAppliedLinearAccClamp=<3000 3000 3000>
    overrideWorldGravity {z=-980}
    simSpaceSettings {dampingAlpha=0, maxAngularAcceleration=500, maxAngularVelocity=3, maxLinearAcceleration=250, maxLinearVelocity=250, worldAlpha=0.5}
   }
   n1 = Use cached pose 'Game/FE'()
   n2 = RigidBody(ComponentPose="Transform (Modify) Bone - Bone: pack"(ComponentPose=Local To Component(LocalPose=@n1), Translation="0,-0.5,0", Rotation="0,0,0") {translationMode=BMM_Additive, translationSpace=BCS_BoneSpace, boneToModify {boneName=pack}, translation {y=-0.5}}) {
    bOverrideWorldGravity=true
    overridePhysicsAsset=&M_MED_MossyDeport_Pack_Physics_ropeFE
    simulationSpace=BaseBoneSpace
    baseBoneRef {boneName=pack}
    componentAppliedLinearAccClamp=<3000 3000 3000>
    overrideWorldGravity {z=-980}
    simSpaceSettings {dampingAlpha=0, maxAngularAcceleration=500, maxAngularVelocity=3, maxLinearAcceleration=250, maxLinearVelocity=250, worldAlpha=0.5}
   }
   n3 = Use cached pose 'Game/FE'()
   n4 = "Transform (Modify) Bone - Bone: pack"(
    ComponentPose="Transform (Modify) Bone - Bone: pack"(ComponentPose=Local To Component(LocalPose=@n3), Translation="0,-0.5,0", Rotation="0,0,0") {translationMode=BMM_Additive, translationSpace=BCS_BoneSpace, boneToModify {boneName=pack}, translation {y=-0.5}}
    Translation=Get backpack_translation()
    Rotation=Get backpack_rotate()
   ) {rotationMode=BMM_Additive, translationMode=BMM_Additive, alphaBoolBlend {blendOption=Linear}, boneToModify {boneName=pack}}
   n5 = RigidBody!(
    ComponentPose=@n4
    Alpha
    OverrideWorldGravity:Vector="0,0,0"
    ExternalForce
    SimSpaceSettings:Sim Space Settings Structure=(WorldAlpha=1,VelocityScaleZ=1,DampingAlpha=0,MaxLinearVelocity=508.1202,MaxAngularVelocity=3,MaxLinearAcceleration=156.2653,MaxAngularAcceleration=750,ExternalLinearDragV=(X=0,Y=0,Z=0),ExternalLinearVelocity=(X=0,Y=0,Z=0),ExternalAngularVelocity=(X=0,Y=0,Z=0))
   ) -> (Pose) {
    bOverrideWorldGravity=true
    overridePhysicsAsset=&/M_MED_MossyDeport_Pack_Physics_rope
    simulationSpace=BaseBoneSpace
    baseBoneRef {boneName=root}
    simSpaceSettings {dampingAlpha=0, maxAngularAcceleration=750, maxAngularVelocity=3, maxLinearAcceleration=156.2653, maxLinearVelocity=508.1202, worldAlpha=1}
   }
   Output Pose(Result=Blend Poses by bool(BlendPose_0=Component To Local(ComponentPose=@n2), BlendPose_1=Component To Local(ComponentPose=@n5), bActiveValue=Get bIsInFrontEnd())) {layerGroup=None}
   Save cached pose 'Game/FE'(Pose=Input Pose()) {cacheName=Game/FE}
  ]}
 EventGraph {nodes [
   Event BlueprintUpdateAnimation()
   > n1 = "Gravity Override State Function (Parameter Asset) (Nativized)"!(
    execute
    CharacterPartAnimInstance=Self-Reference()
    DebugDraw
    GameTimeInSeconds=GetGameTimeInSeconds()
    DebugJointName
    StateLogicLayerName
    Parameters=&/M_MED_MossyDeport_Pack_Parameters
   ) -> (
    then
    OutGravityOverride
    OutCurrentState
    OutGravityOverrideParameters
    OutRigidBodyAnimNodeParameters_ComponentLinearAccScale:Vector=~
    OutRigidBodyAnimNodeParameters_ComponentLinearVelScale:Vector=~
    OutRigidBodyAnimNodeParameters_SimSpaceSettings:Sim Space Settings Structure
   )
   > Set GravityOverride(GravityOverride=@n1)
   > Set Sim Space Settings(Sim Space Settings:Sim Space Settings Structure=@n1.OutRigidBodyAnimNodeParameters_SimSpaceSettings) -> (Output_Get:Sim Space Settings Structure)
   n2 = GetOwnerSkeletalMeshComponent()
   n3 = BreakRotator(InRot=BreakTransform(InTransform=GetDeltaTransformFromRefPose(self=@n2, BoneName=spine_05, BaseName=spine_01)).Rotation)
   n4 = MapRangeClamped(Value=@n3.Yaw, InRangeA=-10, InRangeB=-50, OutRangeB=55)
   > Set backpack_rotate(backpack_rotate=MakeRotator(Roll=@n4, Pitch=float / float(A=@n4, B=10)))
   > Set backpack_translation(backpack_translation=MakeVector(Y=MapRangeClamped(Value=@n3.Yaw, InRangeA=-15, InRangeB=-55, OutRangeB=-8.2), Z=2.5))
   TryGetPawnOwner()
  ]}
]
variables [
 backpack_rotate {type=Rotator}
 backpack_translation {type=Vector}
 GravityOverride {type=Vector}
 Sim Space Settings {
  type=SimSpaceSettings
  default {
   dampingAlpha=1
   maxAngularAcceleration=10000
   maxAngularVelocity=10000
   maxLinearAcceleration=10000
   maxLinearVelocity=10000
   velocityScaleZ=1
   worldAlpha=0
   externalLinearDragV=<0 0 0>
  }
 }
]
